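# Masked material for the MildSE vehicle cosmetic wheel, layering animated glow onto MF_WheelParent_Simple before damage and de-rez transition functions.
# Vertex red controls where custom emissive replaces the parent's, and a rim LOD texture blends opacity mask toward a scalar.
# Named reroutes build a box-projected parallax cubemap reflection, a time-driven hex ripple from rotated UVs, and tinted mask-plus-Fresnel emissive.
# A Glow switch branch is left unconnected.
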Material /VehicleCosmetics/Wheels/MildSE/Materials/M_Wheel_MIldSE
graph {
 properties {
  bAutomaticallySetUsageInEditor=false
  bUseMaterialAttributes=true
  bUsedWithMeshDeformer=true
  bUsedWithSkeletalMesh=true
  blendMode=BLEND_Masked
  pixelDepthOffsetMode=PDOM_Legacy
  previewMesh=/VehicleCosmetics/Wheels/Mild/Meshes/SK_Wheel_Mild
  usedShadingModels=Default Lit
 }
 nodes [
  n1 = MaterialFunctionCall() {materialFunction=&MF_WheelParent_Simple}
  n2 = GetMaterialAttributes(None=@n1.MaterialAttributes) {attributeGetTypes [#1, #2]}
  n3 = OneMinus(None=@n1.VertexColorR)
  n4 = LinearInterpolate(A=@n2.Emissive Color, B=NamedRerouteUsage(), Alpha=@n3)
  n5 = LinearInterpolate(
   A=@n2.Opacity Mask
   B=ScalarParameter() {defaultValue=1, group=5_Special Effects, parameterName=LOds}
   Alpha=OneMinus(None=TextureSample() {samplerType=SAMPLERTYPE_LinearColor, texture=/VehicleCosmetics/Wheels/MildSE/Textures/T_MildSE_Rim_M_LOD}.R)
  )
  MaterialFunctionCall(In=MaterialFunctionCall(Material=SetMaterialAttributes!(MaterialAttributes=@n2.MaterialAttributes, Emissive Color=@n4, Opacity Mask=@n5) {attributeSetTypes [#1, #2]}) {materialFunction=&MF_BaseVehicle_Damage}.Result) {materialFunction=&MF_VehicleDeRez_Transition}
  n6 = Constant3Vector() {constant=<rgba 5000 7500 5000 0>}
  n7 = WorldPosition()
  n8 = Transform(None=ReflectionVectorWS())
  n9 = Subtract(B=@n6) {constA=0}
  n10 = Max(A=Divide(A=Subtract(A=@n6, B=@n7.XYZ), B=@n8), B=Divide(A=Subtract(A=@n9, B=@n7.XYZ), B=@n8))
  n11 = Min(A=ComponentMask(None=@n10) {r=true}, B=ComponentMask(None=@n10) {g=true})
  n12 = ComponentMask(None=@n10) {b=true}
  n13 = Add(A=WorldPosition().XYZ, B=Multiply(A=Min(A=@n11, B=@n12), B=@n8))
  n14 = TextureObject() {samplerType=SAMPLERTYPE_LinearColor, texture=/VehicleCosmetics/Wheels/DuelingDragonsSE/Materials/TC_DuelingDragonsSE_GrassCube}
  n15 = VectorParameter() {group=5_Special Effects, parameterName=Parallex_Color, sortPriority=33, defaultValue=<rgba 0.2948 0.2948 0.2948 1>}
  n16 = Desaturation(
   None=Multiply(A=TextureSample(UVs=@n13, Tex=@n14) {texture=/VehicleCosmetics/Textures/T_White}.RGB, B=@n15.RGB)
   Fraction=ScalarParameter() {defaultValue=0.8, group=5_Special Effects, parameterName=ParallexDesatturation, sortPriority=33}
  ) {luminanceFactors=<rgba 0.3 0.59 0.11 0>}
  n17 = ScalarParameter() {defaultValue=1.5, group=5_Special Effects, parameterName=Parallex_Power, sortPriority=33}
  NamedRerouteDeclaration(None=Saturate(None=Power(Base=@n16, Exp=@n17))) {name=Parallex, nodeColor=<rgba 0.5234 0 1 1>}
  n18 = TextureCoordinate()
  n19 = ComponentMask(None=@n18) {r=true}
  n20 = ComponentMask(None=Rotator(Coordinate=@n18, Time=Constant() {r=2.093}) {speed=1}) {r=true}
  n21 = Rotator(Coordinate=@n18, Time=Constant() {r=4.1888}) {speed=1}
  n22 = ConstantBiasScale(None=AppendVector(A=AppendVector(A=@n19, B=@n20), B=ComponentMask(None=@n21) {r=true})) {bias=-0.5, scale=2}
  n23 = Multiply(A=OneMinus(None=Abs(None=@n22)), B=ScalarParameter() {defaultValue=1, group=5_Special Effects, parameterName=HexGlowTile})
  n24 = Min(A=ComponentMask(None=@n23) {r=true}, B=ComponentMask(None=@n23) {g=true})
  n25 = ComponentMask(None=@n23) {b=true}
  n26 = Time()
  n27 = ScalarParameter() {defaultValue=1, group=5_Special Effects, parameterName=RippleSpeed}
  n28 = Multiply(A=Add(A=Min(A=@n24, B=@n25), B=Multiply(A=@n26, B=@n27)), B=Constant() {r=1})
  n29 = Constant() {r=2}
  n30 = Abs(None=Add(A=Fmod(A=@n28, B=@n29), B=Constant() {r=-1}))
  n31 = Clamp(None=Clamp(None=ConstantBiasScale(None=@n30) {bias=-0.5, scale=2}))
  NamedRerouteDeclaration(None=Clamp(None=Multiply(A=@n31, B=@n31))) {name=MaskColor, nodeColor=<rgba 0.4137 1 0 1>}
  n32 = TextureSampleParameter2D() {
   group=5_Special Effects
   parameterName=Emissive_Mask
   samplerType=SAMPLERTYPE_LinearColor
   texture=/VehicleCosmetics/Wheels/MildSE/Textures/T_MildSE_Rim_E
  }
  n33 = Clamp(None=@n32.RGB)
  n34 = Multiply(A=Constant() {r=0.35}, B=@n33)
  n35 = Multiply(A=@n33, B=Constant() {r=2})
  n36 = NamedRerouteUsage()
  n37 = VectorParameter() {group=5_Special Effects, parameterName=CustomColor, defaultValue=<rgba 1 0.0999 0.0999 1>}
  n38 = Multiply(
   A=Multiply(A=LinearInterpolate(A=@n34, B=@n35, Alpha=@n36), B=@n37.RGB)
   B=ScalarParameter() {defaultValue=3, group=5_Special Effects, parameterName=EmissiveBrightness}
  )
  n39 = TextureSampleParameter2D() {
   group=5_Special Effects
   parameterName=Rim_Normal_Effect
   samplerType=SAMPLERTYPE_Normal
   texture=/VehicleCosmetics/Wheels/MildSE/Textures/T_MildSE_Rim_N
  }
  n40 = MaterialFunctionCall!(Normal=@n39.RGB, Exponent=Constant() {r=4}) {materialFunction=/VehicleCosmetics/Wheels/Cola_Holographic/Materials/MF_Fresnel_Classic}
  n41 = Desaturation(None=@n37.RGB, Fraction=Constant() {r=0.2}) {luminanceFactors=<rgba 0.3 0.59 0.11 1>}
  n42 = Add(A=Multiply(A=Multiply(A=@n37.RGB, B=@n40.Result), B=Multiply(A=@n39.B, B=@n41)), B=Multiply(A=ReflectionVectorWS(), B=Constant()))
  NamedRerouteDeclaration(None=Add(A=Add(A=@n38, B=@n42), B=NamedRerouteUsage())) {name=Emissive, nodeColor=<rgba 1 0.8672 0 1>}
  ScalarParameter() {group=5_Special Effects, parameterName=Roughness}
  ScalarParameter() {defaultValue=1, group=5_Special Effects, parameterName=Metallic}
  n43 = LinearInterpolate(
   A=Multiply(
    A=TextureSample() {samplerType=SAMPLERTYPE_LinearColor, texture=/VehicleCosmetics/Wheels/MildSE/Textures/T_MildSE_Rim_RGB}.B
    B=Constant() {r=0.75}
   )
   B=Constant() {r=0.5}
   Alpha=VertexColor().A
  )
  n44 = Constant() {r=1}
  n45 = Constant()
  n46 = Multiply(
   A=Power(
    Base=MaterialFunctionCall!(Desaturate=@n44, Tint=@n45, TintAmount=@n45, ParallaxSize) {materialFunction=/VehicleCosmetics/Wheels/Mild/Materials/MF/MF_ParallexCorrectWorldENV}.WorldSpace
    Exp=Constant() {r=1.5}
   )
   B=Constant() {r=0.5}
  )
  StaticSwitchParameter(False=LinearInterpolate(B=Multiply(A=@n43, B=@n46), Alpha=@n3)) {group=5_Special Effects, parameterName=Glow}
 ]
}
